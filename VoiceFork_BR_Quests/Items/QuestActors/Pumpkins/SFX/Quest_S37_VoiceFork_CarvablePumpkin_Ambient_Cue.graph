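# Sound cue for the looping ambient sound of a carvable pumpkin quest actor in the VoiceFork Battle Royale quests.
# A Random node picks one of two looping waves, then a Modulator varies its pitch before it reaches a single-input Mixer.
# It borrows a hidden-star supply drop attenuation asset and uses the gadgets sound class.
# A concurrency override allows one instance per owner, stopping the oldest.
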SoundCue /VoiceFork_BR_Quests/Items/QuestActors/Pumpkins/SFX/Quest_S37_VoiceFork_CarvablePumpkin_Ambient_Cue
root=Mixer_0
nodes [
 {id=Mixer_0, children [Modulator_0], properties {inputVolume [1]}}
 {id=Modulator_0, children [Random_0], properties {pitchMin=0.9}}
 {id=Random_0, children [WavePlayer_14, WavePlayer_15], properties {weights [1,1]}}
 {
  id=WavePlayer_14
  properties {bLooping=true, soundWaveAssetPtr=/VoiceFork_BR_Quests/Items/QuestActors/Pumpkins/SFX/Waves/Quest_S37_VoiceFork_CarvablePumpkin_Amb_Loop_01}
 }
 {
  id=WavePlayer_15
  properties {bLooping=true, soundWaveAssetPtr=/VoiceFork_BR_Quests/Items/QuestActors/Pumpkins/SFX/Waves/Quest_S37_VoiceFork_CarvablePumpkin_Amb_Loop_02}
 }
]
properties {
 attenuationSettings=/Game/Athena/Sounds/SupplyDrops/Fort_HiddenStar_Ambient_Loop_Att
 bOverrideConcurrency=true
 maxDistance=2500
 soundClassObject=&Fort_Gadgets
 volumeMultiplier=1.25
 concurrencyOverrides {bLimitToOwner=true, maxCount=1, resolutionRule=StopOldest}
}
