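# Sound cue for carving the carvable pumpkin quest actor in the Season 37 VoiceFork quest content.
# A Random node picks among five equally weighted carve waves and feeds a Mixer at reduced volume, whose second input is empty.
# It uses the gadgets sound class, borrows attenuation from a supply-drop ambient loop asset, and overrides concurrency to one instance per owner, stopping the oldest.
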SoundCue /VoiceFork_BR_Quests/Items/QuestActors/Pumpkins/SFX/Quest_S37_VoiceFork_CarvablePumpkin_Carve_Cue
root=Mixer_0
nodes [
 {id=Mixer_0, children [Random_0,null], properties {inputVolume [0.25,2]}}
 {id=Random_0, children [WavePlayer_28, WavePlayer_29, WavePlayer_30, WavePlayer_31, WavePlayer_32], properties {weights [1,1,1,1,1]}}
 {
  id=WavePlayer_28
  properties {soundWaveAssetPtr=/VoiceFork_BR_Quests/Items/QuestActors/Pumpkins/SFX/Waves/Quest_S37_VoiceFork_CarvablePumpkin_Carve_01}
 }
 {
  id=WavePlayer_29
  properties {soundWaveAssetPtr=/VoiceFork_BR_Quests/Items/QuestActors/Pumpkins/SFX/Waves/Quest_S37_VoiceFork_CarvablePumpkin_Carve_02}
 }
 {
  id=WavePlayer_30
  properties {soundWaveAssetPtr=/VoiceFork_BR_Quests/Items/QuestActors/Pumpkins/SFX/Waves/Quest_S37_VoiceFork_CarvablePumpkin_Carve_03}
 }
 {
  id=WavePlayer_31
  properties {soundWaveAssetPtr=/VoiceFork_BR_Quests/Items/QuestActors/Pumpkins/SFX/Waves/Quest_S37_VoiceFork_CarvablePumpkin_Carve_04}
 }
 {
  id=WavePlayer_32
  properties {soundWaveAssetPtr=/VoiceFork_BR_Quests/Items/QuestActors/Pumpkins/SFX/Waves/Quest_S37_VoiceFork_CarvablePumpkin_Carve_05}
 }
]
properties {
 attenuationSettings=/Game/Athena/Sounds/SupplyDrops/Fort_HiddenStar_Ambient_Loop_Att
 bOverrideConcurrency=true
 maxDistance=2500
 soundClassObject=&Fort_Gadgets
 volumeMultiplier=1
 concurrencyOverrides {bLimitToOwner=true, maxCount=1, resolutionRule=StopOldest}
}
