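# Sound cue for the Season 37 carvable pumpkin quest actor's layered laugh.
# A Mixer blends four randomized layers at different volumes: Laugh A, Laugh B through a default Modulator, and two independent picks from Laugh C.
# One C layer starts after a short random delay, while the other is pitched up to thicken the laugh.
# Borrowed supply-drop attenuation applies, and per-owner concurrency keeps one instance, stopping the oldest.
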SoundCue /VoiceFork_BR_Quests/Items/QuestActors/Pumpkins/SFX/Quest_S37_VoiceFork_CarvablePumpkin_Cue
root=Mixer_0
nodes [
 {id=Delay_1, children [Random_4], properties {delayMax=0.2, delayMin=0.1}}
 {id=Mixer_0, children [Random_0,Modulator_0,Delay_1,Modulator_2], properties {inputVolume [0.75,1.25,1,1.5]}}
 {id=Modulator_0, children [Random_1]}
 {id=Modulator_2, children [Random_3], properties {pitchMax=1.5, pitchMin=1.5}}
 {id=Random_0, children [WavePlayer_0,WavePlayer_1,WavePlayer_2,WavePlayer_3,WavePlayer_4], properties {weights [1,1,1,1,1]}}
 {id=Random_1, children [WavePlayer_6, WavePlayer_7, WavePlayer_8, WavePlayer_9, WavePlayer_10], properties {weights [1,1,1,1,1]}}
 {id=Random_3, children [WavePlayer_11, WavePlayer_12, WavePlayer_18, WavePlayer_29, WavePlayer_30], properties {weights [1,1,1,1,1]}}
 {id=Random_4, children [WavePlayer_19, WavePlayer_20, WavePlayer_21, WavePlayer_27, WavePlayer_28], properties {weights [1,1,1,1,1]}}
 {
  id=WavePlayer_0
  properties {soundWaveAssetPtr=/VoiceFork_BR_Quests/Items/QuestActors/Pumpkins/SFX/Waves/Quest_S37_VoiceFork_CarvablePumpkin_Laugh_A_01}
 }
 {
  id=WavePlayer_1
  properties {soundWaveAssetPtr=/VoiceFork_BR_Quests/Items/QuestActors/Pumpkins/SFX/Waves/Quest_S37_VoiceFork_CarvablePumpkin_Laugh_A_02}
 }
 {
  id=WavePlayer_10
  properties {soundWaveAssetPtr=/VoiceFork_BR_Quests/Items/QuestActors/Pumpkins/SFX/Waves/Quest_S37_VoiceFork_CarvablePumpkin_Laugh_B_05}
 }
 {
  id=WavePlayer_11
  properties {soundWaveAssetPtr=/VoiceFork_BR_Quests/Items/QuestActors/Pumpkins/SFX/Waves/Quest_S37_VoiceFork_CarvablePumpkin_Laugh_C_01}
 }
 {
  id=WavePlayer_12
  properties {soundWaveAssetPtr=/VoiceFork_BR_Quests/Items/QuestActors/Pumpkins/SFX/Waves/Quest_S37_VoiceFork_CarvablePumpkin_Laugh_C_02}
 }
 {
  id=WavePlayer_18
  properties {soundWaveAssetPtr=/VoiceFork_BR_Quests/Items/QuestActors/Pumpkins/SFX/Waves/Quest_S37_VoiceFork_CarvablePumpkin_Laugh_C_03}
 }
 {
  id=WavePlayer_19
  properties {soundWaveAssetPtr=&Quest_S37_VoiceFork_CarvablePumpkin_Laugh_C_01}
 }
 {
  id=WavePlayer_2
  properties {soundWaveAssetPtr=/VoiceFork_BR_Quests/Items/QuestActors/Pumpkins/SFX/Waves/Quest_S37_VoiceFork_CarvablePumpkin_Laugh_A_03}
 }
 {
  id=WavePlayer_20
  properties {soundWaveAssetPtr=&Quest_S37_VoiceFork_CarvablePumpkin_Laugh_C_02}
 }
 {
  id=WavePlayer_21
  properties {soundWaveAssetPtr=&Quest_S37_VoiceFork_CarvablePumpkin_Laugh_C_03}
 }
 {
  id=WavePlayer_27
  properties {soundWaveAssetPtr=/VoiceFork_BR_Quests/Items/QuestActors/Pumpkins/SFX/Waves/Quest_S37_VoiceFork_CarvablePumpkin_Laugh_C_04}
 }
 {
  id=WavePlayer_28
  properties {soundWaveAssetPtr=/VoiceFork_BR_Quests/Items/QuestActors/Pumpkins/SFX/Waves/Quest_S37_VoiceFork_CarvablePumpkin_Laugh_C_05}
 }
 {
  id=WavePlayer_29
  properties {soundWaveAssetPtr=&Quest_S37_VoiceFork_CarvablePumpkin_Laugh_C_04}
 }
 {
  id=WavePlayer_3
  properties {soundWaveAssetPtr=/VoiceFork_BR_Quests/Items/QuestActors/Pumpkins/SFX/Waves/Quest_S37_VoiceFork_CarvablePumpkin_Laugh_A_04}
 }
 {
  id=WavePlayer_30
  properties {soundWaveAssetPtr=&Quest_S37_VoiceFork_CarvablePumpkin_Laugh_C_05}
 }
 {
  id=WavePlayer_4
  properties {soundWaveAssetPtr=/VoiceFork_BR_Quests/Items/QuestActors/Pumpkins/SFX/Waves/Quest_S37_VoiceFork_CarvablePumpkin_Laugh_A_05}
 }
 {
  id=WavePlayer_6
  properties {soundWaveAssetPtr=/VoiceFork_BR_Quests/Items/QuestActors/Pumpkins/SFX/Waves/Quest_S37_VoiceFork_CarvablePumpkin_Laugh_B_01}
 }
 {
  id=WavePlayer_7
  properties {soundWaveAssetPtr=/VoiceFork_BR_Quests/Items/QuestActors/Pumpkins/SFX/Waves/Quest_S37_VoiceFork_CarvablePumpkin_Laugh_B_02}
 }
 {
  id=WavePlayer_8
  properties {soundWaveAssetPtr=/VoiceFork_BR_Quests/Items/QuestActors/Pumpkins/SFX/Waves/Quest_S37_VoiceFork_CarvablePumpkin_Laugh_B_03}
 }
 {
  id=WavePlayer_9
  properties {soundWaveAssetPtr=/VoiceFork_BR_Quests/Items/QuestActors/Pumpkins/SFX/Waves/Quest_S37_VoiceFork_CarvablePumpkin_Laugh_B_04}
 }
]
properties {
 attenuationSettings=/Game/Athena/Sounds/SupplyDrops/Fort_HiddenStar_Ambient_Loop_Att
 bOverrideConcurrency=true
 maxDistance=2500
 soundClassObject=&Fort_Gadgets
 volumeMultiplier=0.5
 concurrencyOverrides {bLimitToOwner=true, maxCount=1, resolutionRule=StopOldest}
}
